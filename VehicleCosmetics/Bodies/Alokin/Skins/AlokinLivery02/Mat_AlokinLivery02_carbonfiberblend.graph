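# Skinned vehicle body material for the Alokin Livery02 cosmetic, blending shared carbon fiber with the Endit universal skin.
# The red channel of BodyCarbonFiberMask lerps each attribute from carbon fiber to Endit, fading out clear-coat data and flattening its bottom normal where Endit shows.
# The blended attributes pass through body normal LOD and vehicle post-FX functions.
# A separate damage and de-rez transition chain takes no input from this blend.
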Material /VehicleCosmetics/Bodies/Alokin/Skins/AlokinLivery02/Mat_AlokinLivery02_carbonfiberblend
graph {
 properties {
  bAutomaticallySetUsageInEditor=false
  bHasPixelAnimation=true
  bUseMaterialAttributes=true
  bUsedWithMeshDeformer=true
  bUsedWithSkeletalMesh=true
  usedShadingModels=Default Lit
 }
 nodes [
  n1 = GetMaterialAttributes(None=MaterialFunctionCall() {materialFunction=/VehicleCosmetics/SharedMaterials/MF_CarbonFiber_VC}.Result) {attributeGetTypes [
    #1
    #2
    #3
    #4
    #5
    #6
   ]}
  n2 = GetMaterialAttributes(None=MaterialFunctionCall() {materialFunction=/VehicleCosmetics/UniversalSkins/Endit/MF_Skin_Endit}.Result) {attributeGetTypes [
    #1
    #2
    #7
    #4
    #8
    #9
   ]}
  n3 = TextureSampleParameter2D() {parameterName=BodyCarbonFiberMask, texture=&/T_AlokinBlankskinOverride}
  n4 = Constant3Vector() {constant=<rgba 0 0 1 0>}
  n5 = SetMaterialAttributes!(
   MaterialAttributes
   Base Color=LinearInterpolate(A=@n1.Base Color, B=@n2.Base Color, Alpha=@n3.R)
   Metallic=LinearInterpolate(A=@n1.Metallic, B=@n2.Metallic, Alpha=@n3.R)
   Emissive Color=LinearInterpolate(B=@n2.Emissive Color, Alpha=@n3.R)
   Roughness=LinearInterpolate(A=@n1.Roughness, B=@n2.Roughness, Alpha=@n3.R)
   Normal=LinearInterpolate(A=@n4, B=@n2.Normal, Alpha=@n3.R)
   Ambient Occlusion=LinearInterpolate(B=@n2.Ambient Occlusion, Alpha=@n3.R) {constA=1}
   Custom Data 0=LinearInterpolate(A=@n1.Custom Data 0, Alpha=@n3.R) {constB=0}
   Custom Data 1=LinearInterpolate(A=@n1.Custom Data 1, Alpha=@n3.R)
   ClearCoatBottomNormal=LinearInterpolate(A=@n1.ClearCoatBottomNormal, B=@n4, Alpha=@n3.R)
  ) {attributeSetTypes [
    #1
    #2
    #7
    #4
    #8
    #9
    #3
    #5
    #6
   ]}
  MaterialFunctionCall!(In=MaterialFunctionCall!(In=@n5) {materialFunction=/VehicleCosmetics/Materials/MaterialFunctions/MF_BodyNormalLOD}.Result) {materialFunction=/VehicleCosmetics/Materials/MaterialFunctions/MF_Vehicle_FNPostFX}
  MaterialFunctionCall(In=MaterialFunctionCall() {materialFunction=&MF_BaseVehicle_Damage}.Result) {materialFunction=&MF_VehicleDeRez_Transition}
 ]
}
